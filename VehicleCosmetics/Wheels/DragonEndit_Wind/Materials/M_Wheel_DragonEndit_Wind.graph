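# Masked wheel material for the DragonEndit Wind cosmetic, adding animated flame glow to the rim and tire.
# It overrides emissive, base color and normal from the RL124 parent function, then applies vehicle damage and de-rez transition functions.
# Panning flow textures distort flame masks, alongside a panning hue-shift multichrome color.
# Static switches toggle rim and tire lights, tire dust and rubber detail normals.
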Material /VehicleCosmetics/Wheels/DragonEndit_Wind/Materials/M_Wheel_DragonEndit_Wind
graph {
 properties {
  bAutomaticallySetUsageInEditor=false
  bUseMaterialAttributes=true
  bUsedWithMeshDeformer=true
  bUsedWithSkeletalMesh=true
  blendMode=BLEND_Masked
  pixelDepthOffsetMode=PDOM_Legacy
  usedShadingModels=Default Lit
 }
 nodes [
  ComponentMask() {g=true, r=true}
  n1 = GetMaterialAttributes(None=MaterialFunctionCall() {materialFunction=/VehicleCosmetics/Wheels/RL124/Materials/MF_Wheel_RL124_Parent_Simple}.MaterialAttributes) {attributeGetTypes [#1, #2, #3]}
  n2 = NamedRerouteUsage()
  n3 = NamedRerouteUsage()
  n4 = NamedRerouteUsage()
  MaterialFunctionCall(In=MaterialFunctionCall(Material=SetMaterialAttributes!(MaterialAttributes=@n1.MaterialAttributes, Emissive Color=@n2, Normal=@n3, Base Color=@n4) {attributeSetTypes [#4, #1, #2]}) {materialFunction=&MF_BaseVehicle_Damage}.Result) {materialFunction=&MF_VehicleDeRez_Transition}
  n5 = TextureCoordinate() {coordinateIndex=1, uTiling=15, vTiling=15}
  n6 = ScalarParameter() {defaultValue=0.1, group=Fire, parameterName=MultichromeScale}
  n7 = ScalarParameter() {defaultValue=0.2, group=Fire, parameterName=MultichromePanSpeed}
  n8 = Time()
  n9 = ComponentMask(None=Panner(Coordinate=Multiply(A=@n5, B=@n6), Time=Multiply(A=@n7, B=@n8)) {speedX=1, speedY=1}) {r=true}
  n10 = StaticSwitchParameter(
   True=VectorParameter() {group=Fire, parameterName=CustomColor, defaultValue=<rgba 1 0.3282 0 1>}.RGB
   False=VectorParameter() {group=Fire, parameterName=FlameColor, defaultValue=<rgba 0 0 0 1>}.RGB
  ) {defaultValue=true, parameterName=InheritColors}
  NamedRerouteDeclaration(None=LinearInterpolate(
    A=@n10
    B=MaterialFunctionCall(Hue Shift Percentage=@n9, Texture=@n10) {materialFunction=&HueShift}.Result
    Alpha=ScalarParameter() {defaultValue=0.3, group=Fire, parameterName=MultichromeColor}
   )) {name=Multichrome, nodeColor=<rgba 1 0.1778 0.5664 1>}
  NamedRerouteDeclaration(None=VertexColor().R) {name=VC_R, nodeColor=<rgba 1 0.044 0 1>}
  n11 = TextureSampleParameter2D() {
   parameterName=RimFlameMask
   samplerType=SAMPLERTYPE_LinearColor
   texture=/VehicleCosmetics/Wheels/DragonEndit/Textures/T_DragonEndit_Rim_M_LOD
  }
  n12 = Multiply(A=TextureCoordinate() {uTiling=0.5, vTiling=0.3}, B=ScalarParameter() {defaultValue=1, group=Fire, parameterName=UVDistortTile})
  n13 = TextureSample(UVs=Panner(Coordinate=@n12) {speedX=0.15, speedY=0.3}) {samplerType=SAMPLERTYPE_LinearColor, texture=/VehicleCosmetics/Wheels/Dragon/Textures/T_Flow_03}
  n14 = ScalarParameter() {defaultValue=0.5, group=Fire, parameterName=UVDistoertAmt}
  n15 = Multiply(A=TextureCoordinate() {uTiling=6, unMirrorV=true, vTiling=3}, B=Multiply(A=ComponentMask(None=@n13.RGB) {g=true, r=true}, B=@n14))
  n16 = NamedRerouteUsage()
  n17 = TextureCoordinate() {vTiling=4}
  n18 = Panner(Coordinate=@n12) {speedX=0.25, speedY=0.5}
  n19 = Multiply(
   A=ComponentMask(None=TextureSample(UVs=@n18) {samplerType=SAMPLERTYPE_LinearColor, texture=&T_Flow_03}.RGB) {g=true, r=true}
   B=@n14
  )
  n20 = Multiply(
   A=TextureSampleParameter2D(UVs=Multiply(A=@n17, B=@n19)) {parameterName=Tire_Mask, texture=/VehicleCosmetics/Wheels/Dragon/Textures/T_Tire_Flame_M}.R
   B=Clamp(None=NamedRerouteUsage())
  )
  n21 = StaticSwitchParameter(
   True=Multiply(
    A=TextureSampleParameter2D(UVs=@n15) {parameterName=Tire_Mask, texture=&T_Tire_Flame_M}.R
    B=Clamp(None=@n16)
   )
   False=@n20
  ) {group=Fire, parameterName=LODRimTiling}
  n22 = Multiply(A=Multiply(A=@n11.G, B=Multiply(A=@n21) {constB=5}), B=ScalarParameter() {defaultValue=3, group=Fire, parameterName=Emissive_Rim})
  n23 = Constant()
  n24 = TextureCoordinate() {uTiling=0.5}
  n25 = TextureSample(UVs=@n24) {texture=/VehicleCosmetics/Wheels/_SharedTextures/FX/FlameTire_M_T}
  n26 = Add(A=@n25.R, B=@n25.R)
  n27 = Clamp(None=OneMinus(None=@n26))
  n28 = TextureSample() {samplerType=SAMPLERTYPE_LinearColor, texture=/VehicleCosmetics/Wheels/DragonEndit/Textures/T_DragonEndit_Tire_M}
  n29 = Multiply(A=Multiply(A=@n28.G, B=Multiply(A=@n25.B, B=@n27)), B=NamedRerouteUsage())
  n30 = ScalarParameter() {defaultValue=10, group=Fire, parameterName=EmissiveTread}
  n31 = Add(A=@n28.G, B=@n25.B)
  n32 = TextureSample(UVs=Panner(Coordinate=@n24) {speedX=0.1}) {texture=&FlameTire_M_T}
  n33 = Multiply(A=@n32.B, B=Clamp(None=Subtract(A=@n31, B=@n26)))
  n34 = Multiply(A=Multiply(A=NamedRerouteUsage(), B=@n32.R), B=Constant() {r=3})
  n35 = Add(A=Add(A=Multiply(A=@n29, B=@n30), B=@n20), B=Multiply(A=Multiply(A=@n34, B=@n33), B=@n30))
  NamedRerouteDeclaration(None=LinearInterpolate(
    A=StaticSwitchParameter(True=@n22, False=@n23) {defaultValue=true, group=Fire, parameterName=RimLights}
    B=StaticSwitchParameter(True=@n35, False=@n23) {defaultValue=true, group=Fire, parameterName=TireLights}
    Alpha=NamedRerouteUsage()
   )) {name=Emissive, nodeColor=<rgba 1 0.3961 0.5622 1>}
  NamedRerouteDeclaration(None=@n1.Normal) {name=NormalIN, nodeColor=<rgba 0 0.1328 1 1>}
  n36 = NamedRerouteUsage()
  n37 = TextureCoordinate()
  n38 = ScalarParameter() {defaultValue=1, group=TireRubber, parameterName=TextureDetailNormalTiling}
  n39 = Multiply(
   A=TextureSample(UVs=Multiply(A=@n37, B=@n38)) {samplerType=SAMPLERTYPE_Normal, texture=/VehicleCosmetics/SharedTextures/T_Matte_N}.RGB
   B=NamedRerouteUsage()
  )
  NamedRerouteDeclaration(None=StaticSwitchParameter(True=Add(A=@n36, B=@n39), False=@n36) {group=TireRubber, parameterName=TireRubberExist}) {name=NormalRubber, nodeColor=<rgba 0.1016 0 1 1>}
  n40 = NamedRerouteUsage()
  n41 = Multiply(A=VectorParameter() {group=Fire, parameterName=TireDustColor, defaultValue=<rgba 0 0 0 1>}.RGB, B=NamedRerouteUsage())
  n42 = Multiply(A=@n26, B=NamedRerouteUsage())
  NamedRerouteDeclaration(None=StaticSwitchParameter(True=LinearInterpolate(A=@n40, B=@n41, Alpha=@n42), False=NamedRerouteUsage()) {defaultValue=true, group=Fire, parameterName=TireDustExist}) {name=BaseColor, nodeColor=<rgba 0.3359 0 1 1>}
  NamedRerouteDeclaration(None=@n1.Base Color) {name=BaseColor_IN, nodeColor=<rgba 1 0.0938 0 1>}
 ]
}
